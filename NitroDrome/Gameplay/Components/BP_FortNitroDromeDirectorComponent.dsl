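; Fortnite NitroDrome director component that drives arena reactions: effects, screens and loudspeaker audio.
; A repeating timer waits for World Partition streaming before caching the FX manager actor.
; Event router listeners rebroadcast fire pit hazards as alarm events and flag boss spawns.
; Triggered events cosmetically activate all FX and map tags to screen states, while sounds play at the loudspeaker, or gameplay volume as fallback, with a boss-spawn parameter.
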(event Custom|Streaming%20Check
  (bind _returnvalue (Streaming|IsAllStreamingCompleted (WorldSubsystems|GetWorldPartitionSubsystem)))
  (if _returnvalue
    (Utilities|Time|ClearandInvalidateTimerbyHandle (Variables|Default|GetStreamingTimer))
    (bind _outactors (Actor|GetAllActorsOfClass "/Game/Environments/Helios/FX/NitroDrome/BP/BP_NitroDrome_Fancy_Mananger.BP_NitroDrome_Fancy_Mananger_C"))
    (if (Utilities|Array|IsValidIndex _outactors)
      (Variables|Default|SetFXManager (Utilities|Array|Get(acopy) _outactors)))))

(event EventBeginPlay
  (|CallParentFunction)
  (bind _returnvalue (Utilities|Time|SetTimerbyEvent (EventDispatchers|CreateEvent :!event Custom|Streaming%20Check) 0.1 true))
  (Variables|Default|SetStreamingTimer _returnvalue)
  (bind _returnvalue_2 (GameplayEventRouter|GetEventRouter "/NitroDrome/Gameplay/Actors/BP_NitroDromePlayspace.BP_NitroDromePlayspace_C" (Components|GetOwner)))
  (Utilities|FlowControl|Sequence
    (:then_0
      (GameplayEventRouter|StartListeningToEvent _returnvalue_2 "/FirePitHazard/Gameplay/FirePitHazardEventStruct.FirePitHazardEventStruct"
        (:OnEventReceived
          (CallFunction|BroadcastNitroDromeReactableEvent (GameplayTags|MakeLiteralGameplayTagContainer "(GameplayTags=((TagName=\"NitroDrome.Event.FirePitHazard.Alarm\")))")))))
    (:then_1
      (bind (_asynctaskproxy _payload) (GameplayEventRouter|StartListeningToEvent _returnvalue_2 "/Script/NitroDromeRuntime.OnNitroDromeEventTriggered")
        (:OnEventReceived
          (bind _payload_eventtag (Utilities|Struct|BreakOnNitroDromeEventTriggered _payload))
          (if (GameplayTags|MatchesTag _payload_eventtag "(TagName=\"NitroDrome.Event.Boss.Spawn\")" true)
            (Variables|Default|SetNitroQueenPresent true)))))))

(event EventTick
  (params (DeltaSeconds :type Float%20%28single-precision%29)))

(event EventOnEventTriggered
  (params (EventTag :type Gameplay%20Tag%20Structure))
  (bind _fxmanager (Variables|Default|GetFXManager))
  (Variables|Default|SetEvent%20Tag EventTag)
  (Networking|CanExecuteCosmeticEvents
    (:True
      ;[1 0.546 0] Activating on ANY reaction as requested by FX. Actual visibility is controlled by the Param collections driven by the Reactions table.
      ; Ideally actual activation of these is driven by the table but we want to get the sign off before committing to any approaches
      (Utilities|IsValid _fxmanager
        (:Is%20Valid
          (Class|BPNitroDromeFancyMananger|ActivateEffects _fxmanager)
          ;[0.6155 0 1] Passing in event for the screens to react to
          (Default|MapEventstoPOIScreens))
        (:Is%20Not%20Valid
          (Fortnite|FortShippingLog self "Make sure you set the FX Manager to somethng Valid in the Gameplay Volume!" true))))))

(event EventPlaySoundOnLoudSpeakers
  (params (Sound :type Sound%20Base%20Object%20Reference))
  (bind _returnvalue (Components|GetOwner))
  (bind _asplayspace (Utilities|Casting|CastToPlayspace _returnvalue))
  (bind _boundgameplayvolume (Variables|/Script/PlayspaceSystem.Playspace|GetBoundGameplayVolume _asplayspace))
  (bind _loudspeaker (CallFunction|GetLoudspeakerfromGameplayVolume))
  (Utilities|IsValid _boundgameplayvolume
    (:Is%20Valid
      (bind _returnvalue_2 (Audio|SpawnSoundatLocation Sound (Math|Vector|SelectVector (Transformation|GetWorldLocation (Variables|/NitroDrome/Assets/Blueprints/BP_NitroDrome_Center_A.BP_NitroDrome_Center_A_C|GetSpeakers _loudspeaker)) (Transformation|GetActorLocation _boundgameplayvolume) (Utilities|IsValid _loudspeaker :!class /Script/Engine.KismetSystemLibrary))))
      (Audio|Parameter|SetBooleanParameter _returnvalue_2 "BossSpawn" (Variables|Default|GetNitroQueenPresent) :!class /Script/AudioExtensions.AudioParameterControllerInterface))))

(fn GetFXManager
  (Utilities|IsValid (Variables|Default|GetFXManager)
    (:Is%20Valid
      (return (Variables|Default|GetFXManager)))))

(fn GetLoudspeakerFromGameplayVolume
  (bind _asplayspace (Utilities|Casting|CastToPlayspace (Components|GetOwner)))
  (bind _asnitro_drome_gameplay_volume (Utilities|Casting|CastToNitroDromeGameplayVolume (Variables|/Script/PlayspaceSystem.Playspace|GetBoundGameplayVolume _asplayspace)))
  (bind _loudspeaker (Class|NitroDromeGameplayVolume|GetLoudspeaker _asnitro_drome_gameplay_volume))
  (return _loudspeaker))

(fn MapEventsToPOIScreens
  (bind _event_tag (Variables|Default|GetEvent%20Tag))
  (bind _tagbytemap (Variables|Default|GetTagByteMap))
  (bind (_value _returnvalue) (Utilities|Map|Find _tagbytemap _event_tag))
  (bind _fxmanager (Variables|Default|GetFXManager))
  (if _returnvalue
    (Class|BPNitroDromeFancyMananger|ActivateRustBucketScreens _fxmanager _value)
    (Class|BPNitroDromeFancyMananger|ActivateRustBucketBossScreens _fxmanager 0)))
