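; Playspace component that gathers server-submitted analytics on NitroDrome POI activity, destroying itself when its hotfix flag is off.
; Gameplay volume overlaps record unique humans, bots, teams and human dwell time, and forward vehicles to separate handlers.
; Pawn delegates bound during overlap tally damage inside or outside the POI and deaths caused from within it.
; A game phase listener submits the attributes once at EndGame; EndPlay also submits in editor.
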(event EventBeginPlay
  ;[0.01873 0.1667 0.03335] BeginPlay: On the server, bind delegate to the game state to know when the game phase changes.
  ; On the client, disable collision.
  (|CallParentFunction)
  (if (not (ScalableFloat|GetValueasBool (Variables|Default|GetHF_EnableAnalytics)))
    (Components|DestroyComponent)
    (else
      (Utilities|FlowControl|Sequence
        (:then_0
          (bind _asplayspace (Utilities|Casting|CastToPlayspace (Components|GetOwner)))
          (Collision|BindEventtoOnActorBeginOverlap (Variables|/Script/PlayspaceSystem.Playspace|GetBoundGameplayVolume _asplayspace) (EventDispatchers|CreateEvent :!event Custom|OnPlayspaceBeginOverlap))
          (Collision|BindEventtoOnActorEndOverlap (Variables|/Script/PlayspaceSystem.Playspace|GetBoundGameplayVolume _asplayspace) (EventDispatchers|CreateEvent :!event Custom|OnPlayspaceEndOverlap)))
        (:then_1
          (Class|FortAsyncActionCurrentPlaylistReady|PlaylistDataReadyAsync
            (:OnReady
              (bind (_asynctaskproxy _payload) (GameplayEventRouter|StartListeningToStatefulEvent (GameplayEventRouter|GetGlobalEventRouter) "/Script/FortniteGame.FortGamePhaseUpdatedEvent")
                (:then
                  (Variables|Default|SetGamePhaseUpdatedAsyncTask _asynctaskproxy))
                (:OnEventReceived
                  _)
                (:OnSavedEventStateExists
                  _))
              (CallFunction|HandleGamePhaseChanged (Utilities|Struct|BreakFortGamePhaseUpdatedEvent _payload)))))))))

(event Custom|OnPlayspaceBeginOverlap
  (params (OverlappedActor :type Actor%20Object%20Reference) (OtherActor :type Actor%20Object%20Reference))
  ;[0.03703 0.1469 0.1771] OnComponentBeginOverlap: Start tracking the player that has entered the Poi while tracking values for analytics.
  (bind _asfort_player_pawn_athena (Utilities|Casting|CastToFortPlayerPawnAthena OtherActor)
    (:then
      (bind _asfort_player_state_athena (Utilities|Casting|CastToFortPlayerStateAthena (Variables|/Script/Engine.Pawn|GetPlayerState _asfort_player_pawn_athena)))
      (if (Variables|/Script/Engine.PlayerState|GetbIsABot _asfort_player_state_athena)
        ;;[0.009605 0.05729 0.05043] Track bots and players separately.
        (Utilities|Array|AddUnique (Variables|Default|GetUniquePlayerIds_Bot) (PlayerState|GetPlayerUniqueNetID _asfort_player_state_athena))
        (else
          (Utilities|Array|AddUnique (Variables|Default|GetUniquePlayerIds_Human) (PlayerState|GetPlayerUniqueNetID _asfort_player_state_athena))
          (Utilities|Map|Add (Variables|Default|GetUniquePlayerToTimeEnteredPoiMap) (PlayerState|GetPlayerUniqueNetID _asfort_player_state_athena) (Fortnite|Sync|Time|GetServerWorldTimeSeconds))))
      ;;[0.1354 0.07014 0.03995] Track players that are currently on the Poi.
      (Utilities|Array|AddUnique (Variables|Default|GetOverlappingPlayerPawns) _asfort_player_pawn_athena)
      ;;[0.2395 1 0.5426] Track unique teams that enter the Poi.
      (Utilities|Array|AddUnique (Variables|Default|GetTeams) (Math|Conversions|ToInteger(Byte) (Team|GetActorTeam _asfort_player_state_athena)))
      ;;[0.01975 0.08333 0.06826] Setup delegates to track what the player is doing on the Poi.
      (EventDispatchers|BindEventtoOnDidDamage _asfort_player_pawn_athena (EventDispatchers|CreateEvent :!event Handle%20Pawn%20Did%20Damage))
      (EventDispatchers|BindEventtoOnFortPlayerDied _asfort_player_pawn_athena (EventDispatchers|CreateEvent :!event Handle%20Pawn%20Died))
      (Fortnite|FortShippingLog self (Utilities|String|Append "OnComponentBeginOverlap: " (Fortnite|GetPlayerNameSafe _asfort_player_pawn_athena))))
    (:CastFailed
      (bind _asfort_athena_vehicle (Utilities|Casting|CastToFortAthenaVehicle OtherActor))
      (CallFunction|OnVehicleBeginOverlap _asfort_athena_vehicle))) ; Overlapped a player pawn.
)

(event Custom|OnPlayspaceEndOverlap
  (params (OverlappedActor :type Actor%20Object%20Reference) (OtherActor :type Actor%20Object%20Reference))
  ;[0.1042 0.03434 0.08598] OnComponentEndOverlap: Unbind delegates.
  (bind _asfort_player_pawn_athena (Utilities|Casting|CastToFortPlayerPawnAthena OtherActor)
    (:then
      (Utilities|Array|RemoveItem (Variables|Default|GetOverlappingPlayerPawns) _asfort_player_pawn_athena)
      (EventDispatchers|UnbindEventfromOnDidDamage _asfort_player_pawn_athena (EventDispatchers|CreateEvent :!event Handle%20Pawn%20Did%20Damage))
      (EventDispatchers|UnbindEventfromOnFortPlayerDied _asfort_player_pawn_athena (EventDispatchers|CreateEvent :!event Handle%20Pawn%20Died))
      (Fortnite|FortShippingLog self (Utilities|String|Append "OnComponentEndOverlap: " (Fortnite|GetPlayerNameSafe _asfort_player_pawn_athena)))
      (CallFunction|ComputeTotalDurationforLeavingPlayer _asfort_player_pawn_athena))
    (:CastFailed
      (bind _asfort_athena_vehicle (Utilities|Casting|CastToFortAthenaVehicle OtherActor))
      (CallFunction|OnVehicleEndOverlap _asfort_athena_vehicle))))

(event EventEndPlay
  (params (EndPlayReason :type EEndPlayReason%20Enum :default "Destroyed"))
  ;[0.07292 0.04601 0.006518] EndPlay: Unbind delegates.
  (Utilities|FlowControl|Sequence
    (:then_0
      (CallFunction|FinishNitroDromeSession)
      (Utilities|FlowControl|Sequence
        (:then_0
          (if (Editor|InEditororPIE)
            (CallFunction|SubmitAnalytics)
            (else
              _))
          (bind _gamephaseupdatedasynctask (Variables|Default|GetGamePhaseUpdatedAsyncTask)
            (:then
              (GameplayEventRouter|StopListeningtoStatefulEvent _gamephaseupdatedasynctask))
            (:else)))
        (:then_1
          (Utilities|FlowControl|Sequence
            (:then_0
              (for _array_element (Variables|Default|GetOverlappingPlayerPawns)
                (Utilities|IsValid _array_element
                  (:Is%20Valid
                    (EventDispatchers|UnbindEventfromOnDidDamage _array_element (EventDispatchers|CreateEvent :!event Handle%20Pawn%20Did%20Damage))
                    (EventDispatchers|UnbindEventfromOnFortPlayerDied _array_element (EventDispatchers|CreateEvent :!event Handle%20Pawn%20Died))))))))))
    (:then_1
      (|CallParentFunction EndPlayReason))))

(fn ComputeTotalDurationForLeavingPlayer
  (params (Pawn :type Pawn%20Object%20Reference))
  (bind _returnvalue (Fortnite|Sync|Time|GetServerWorldTimeSeconds))
  (bind _uniqueplayertototaldurationmap (Variables|Default|GetUniquePlayerToTotalDurationMap))
  (bind _uniqueplayertotimeenteredpoimap (Variables|Default|GetUniquePlayerToTimeEnteredPoiMap))
  ;[0.02231 0.1667 0.1584] ComputeTotalDurationForLeavingPlayer: Computes the total duration that the leaving player has spent on the Poi.
  (Utilities|SwitchHasAuthority
    (:Authority
      (Utilities|IsValid Pawn
        (:Is%20Valid
          (bind _asfort_player_state_athena (Utilities|Casting|CastToFortPlayerStateAthena (Variables|/Script/Engine.Pawn|GetPlayerState Pawn)))
          (if (not (Variables|/Script/Engine.PlayerState|GetbIsABot _asfort_player_state_athena))
            (if (.ReturnValue (Utilities|Map|Find _uniqueplayertotimeenteredpoimap (PlayerState|GetPlayerUniqueNetID _asfort_player_state_athena)))
              (if (.ReturnValue (Utilities|Map|Find _uniqueplayertototaldurationmap (PlayerState|GetPlayerUniqueNetID _asfort_player_state_athena)))
                (Utilities|Map|Add (Variables|Default|GetUniquePlayerToTotalDurationMap) (PlayerState|GetPlayerUniqueNetID _asfort_player_state_athena) (+ (Utilities|Map|Find _uniqueplayertototaldurationmap (PlayerState|GetPlayerUniqueNetID _asfort_player_state_athena)) (- _returnvalue (Utilities|Map|Find _uniqueplayertotimeenteredpoimap (PlayerState|GetPlayerUniqueNetID _asfort_player_state_athena)))))
                (else
                  (Utilities|Map|Add (Variables|Default|GetUniquePlayerToTotalDurationMap) (PlayerState|GetPlayerUniqueNetID _asfort_player_state_athena) (- _returnvalue (Utilities|Map|Find _uniqueplayertotimeenteredpoimap (PlayerState|GetPlayerUniqueNetID _asfort_player_state_athena))))))
              (Utilities|Map|Remove (Variables|Default|GetUniquePlayerToTimeEnteredPoiMap) (PlayerState|GetPlayerUniqueNetID _asfort_player_state_athena)))))))))

(fn GetAveragePlayerGameplayTimeOnPoi
  ;[0.008687 0.08909 0.09896] Returns the average time players spent on the Poi.
  (bind _time (CallFunction|GetTotalPlayerGameplayTimeonPoi))
  (return (Math|Float|SafeDivide _time (Math|Conversions|ToFloat(Integer) (Utilities|Map|Length (Variables|Default|GetUniquePlayerToTotalDurationMap))))))

(fn GetFinisherOrDownerPawn
  (params (DyingPlayerPawn :type Fort%20Player%20Pawn%20Athena%20Object%20Reference))
  ;[0.06228 0.1458 0.1458] Returns the player reponsible for downing or eliminating the input player.
  (Utilities|FlowControl|Sequence
    (:then_0
      (bind _playerstate (Variables|/Script/Engine.Pawn|GetPlayerState DyingPlayerPawn)
        (:then
          (bind _asfort_player_state_athena (Utilities|Casting|CastToFortPlayerStateAthena _playerstate))
          (bind _returnvalue_finisherordowner (Utilities|Struct|BreakDeathInfo (Class|FortPlayerStateAthena|GetDeathInfo _asfort_player_state_athena)))
          (Utilities|IsValid _returnvalue_finisherordowner
            (:Is%20Valid
              (Utilities|IsValid (Fortnite|GetControllerfromActor _returnvalue_finisherordowner)
                (:Is%20Valid
                  (bind _asfort_player_pawn_athena (Utilities|Casting|CastToFortPlayerPawnAthena (Pawn|GetControlledPawn (Fortnite|GetControllerfromActor _returnvalue_finisherordowner))))
                  (return _asfort_player_pawn_athena))))))
        (:else)))
    (:then_1
      (return 0))))

(fn GetMidMatchObjectiveAnalyticsAttributeList
  (return (Variables|Default|GetAnalyticsEventAttributes)))

(fn GetTotalPlayerGameplayTimeOnPoi
  (bind _uniqueplayertototaldurationmap (Variables|Default|GetUniquePlayerToTotalDurationMap))
  ;[0.01335 0.09896 0.09601] Sum up the time each unique player spent on the Poi.
  (bind _keys (Utilities|Map|Keys _uniqueplayertototaldurationmap))
  (for _array_element _keys
    (LocalVariables|SetRetTime (+ (LocalVariables|GetRetTime) (Utilities|Map|Find _uniqueplayertototaldurationmap _array_element))))
  (return (LocalVariables|GetRetTime)))

(fn HandlePawnDidDamage
  (params (DamagedActor :type Actor%20Object%20Reference) (Damage :type Float%20%28single-precision%29) (InstigatedBy :type Controller%20Object%20Reference) (DamageCauser :type Actor%20Object%20Reference) (HitLocation :type Vector) (FHitComponent :type Primitive%20Component%20Object%20Reference) (BoneName :type Name) (Momentum :type Vector))
  (bind (_asfort_athena_aibot_controller _bsuccess) (Utilities|Casting|CastToFortAthenaAIBotController InstigatedBy))
  (bind (_asfort_player_controller_athena _bsuccess_2) (Utilities|Casting|CastToFortPlayerControllerAthena InstigatedBy))
  (bind _playerstate (Variables|/Script/Engine.Controller|GetPlayerState InstigatedBy))
  (bind _damage (Variables|GetDamage))
  ;[0.02222 0.1579 0.1927] HandlePawnDidDamage: Called when the player pawn on the Poi does damage.
  (bind _asfort_player_pawn_athena (Utilities|Casting|CastToFortPlayerPawnAthena DamagedActor)) ; Actor damaged is a player pawn?
  (LocalVariables|SetDamagedPawn _asfort_player_pawn_athena)
  (if (or _bsuccess _bsuccess_2)
    (Utilities|IsValid _playerstate
      (:Is%20Valid
        ;;[0.04521 0.1314 0.151] Is the instigator pawn on the Poi?
        (bind _asfort_player_pawn_athena_2 (Utilities|Casting|CastToFortPlayerPawnAthena (PlayerState|GetPawn _playerstate)))
        (if (Utilities|Array|ContainsItem (Variables|Default|GetOverlappingPlayerPawns) _asfort_player_pawn_athena_2)
          ;;[0.05723 0.1979 0.1162] Is the damaged pawn on the Poi?
          (if (Utilities|Array|ContainsItem (Variables|Default|GetOverlappingPlayerPawns) (LocalVariables|GetDamagedPawn))
            (Variables|Default|SetTotalPawnDamageDoneInsidePoi (+ (Variables|Default|GetTotalPawnDamageDoneInsidePoi) _damage))
            (else
              (Variables|Default|SetTotalPawnDamageDoneOutsidePoi (+ _damage (Variables|Default|GetTotalPawnDamageDoneOutsidePoi))))))))))

(fn HandlePawnDied
  (params (FortPlayerPawnAthena :type Fort%20Player%20Pawn%20Athena%20Object%20Reference))
  ;[0.1354 0.03243 0.125] HandlePawnDied: Called whenever a player pawn on the Poi dies.
  ;;[0.1202 0.04357 0.1562] Is dying pawn on the poi?
  (Utilities|IsValid FortPlayerPawnAthena
    (:Is%20Valid
      (if (Utilities|Array|ContainsItem (Variables|Default|GetOverlappingPlayerPawns) FortPlayerPawnAthena)
        ;;[0.01425 0.04515 0.05729] Is the damage instigating player on the Poi?
        (bind _finisherordownerpawn (CallFunction|GetFinisherorDownerPawn FortPlayerPawnAthena))
        (Utilities|IsValid _finisherordownerpawn
          (:Is%20Valid
            (if (Utilities|Array|ContainsItem (Variables|Default|GetOverlappingPlayerPawns) _finisherordownerpawn)
              (Math|Integer|IncrementInt (Variables|Default|GetTotalPawnDeaths)))))))))

(fn HandleGamePhaseChanged
  (params (GamePhase :type EAthenaGamePhase%20Enum))
  ;[0.0164 0.07935 0.1198] HandleGamePhaseChanged: Once we're in EndGame, submit the analytics.
  (if (Utilities|Enum|Equal(Enum) GamePhase "EndGame")
    (CallFunction|FinishNitroDromeSession)
    (CallFunction|SubmitAnalytics)
    (bind _gamephaseupdatedasynctask (Variables|Default|GetGamePhaseUpdatedAsyncTask)
      (:then
        (GameplayEventRouter|StopListeningtoStatefulEvent _gamephaseupdatedasynctask))
      (:else))))

(fn SubmitAnalytics
  (bind _uniqueplayerids_human (Variables|Default|GetUniquePlayerIds_Human))
  (bind _returnvalue (Utilities|Array|Length _uniqueplayerids_human))
  (bind _teams (Variables|Default|GetTeams))
  (bind _returnvalue_2 (Utilities|Array|Length _teams))
  (bind _uniqueplayerids_bot (Variables|Default|GetUniquePlayerIds_Bot))
  (bind _returnvalue_3 (Utilities|Array|Length _uniqueplayerids_bot))
  (bind _returnvalue_4 (CallFunction|WasNitroDromeCaptured))
  (bind _totalpawndeaths (Variables|Default|GetTotalPawnDeaths))
  (bind _totalpawndamagedoneinsidepoi (Variables|Default|GetTotalPawnDamageDoneInsidePoi))
  (bind _totalpawndamagedoneoutsidepoi (Variables|Default|GetTotalPawnDamageDoneOutsidePoi))
  ;[0.04251 0.1354 0.04262] SubmitAnalytics: Fire the analytics!
  (Networking|SwitchHasAuthority
    (:Authority
      (if (and (not (Replay|GetIsPlayingReplay)) (not (Variables|Default|GetbAnalyticsSubmitted)))
        (Variables|Default|SetbAnalyticsSubmitted true)
        (Utilities|FlowControl|Sequence
          (:then_0
            (Utilities|FlowControl|Sequence
              (:then_0
                ;;[0.02121 0.07343 0.08854] At this point, all overlapping players can be considered as players that have left the Poi.
                (for _array_element (Variables|Default|GetOverlappingPlayerPawns)
                  (CallFunction|ComputeTotalDurationforLeavingPlayer _array_element)))
              (:then_1
                (bind _returnvalue_5 (Analytics|ConstructAnalyticsAttribute :AttributeName "UniqueHumanPlayersOnPoi" :IntValue _returnvalue :floatValue -1.0 :BoolValue false))
                (Utilities|Array|Add (Variables|Default|GetAnalyticsEventAttributes) _returnvalue_5)
                (Fortnite|FortShippingLog self (Utilities|String|Append "UniqueHumanPlayersOnPoi: " (Utilities|String|ToString(Integer) _returnvalue)))
                (bind _returnvalue_6 (Analytics|ConstructAnalyticsAttribute :AttributeName "UniqueTeamsOnPoi" :IntValue _returnvalue_2 :floatValue -1.0 :BoolValue false))
                (Utilities|Array|Add (Variables|Default|GetAnalyticsEventAttributes) _returnvalue_6)
                (Fortnite|FortShippingLog self (Utilities|String|Append "UniqueTeamsOnPoi: " (Utilities|String|ToString(Integer) _returnvalue_2)))
                (bind _returnvalue_7 (Analytics|ConstructAnalyticsAttribute :AttributeName "TotalDamageDoneInsidePoi" :IntValue -1 :floatValue _totalpawndamagedoneinsidepoi :BoolValue false))
                (Utilities|Array|Add (Variables|Default|GetAnalyticsEventAttributes) _returnvalue_7)
                (Fortnite|FortShippingLog self (Utilities|String|Append "TotalDamageDoneInsidePoi: " (Utilities|String|ToString(Float) _totalpawndamagedoneinsidepoi)))
                (bind _time (CallFunction|GetTotalPlayerGameplayTimeonPoi))
                (bind _returnvalue_8 (Analytics|ConstructAnalyticsAttribute :AttributeName "TotalPlayerGametimeTimeOnPoi" :IntValue -1 :floatValue _time :BoolValue false))
                (Utilities|Array|Add (Variables|Default|GetAnalyticsEventAttributes) _returnvalue_8)
                (Fortnite|FortShippingLog self (Utilities|String|Append "TotalPlayerGametimeTimeOnPoi: " (Utilities|String|ToString(Float) _time)))
                (bind _time_2 (CallFunction|GetAveragePlayerGameplayTimeonPoi))
                (bind _returnvalue_9 (Analytics|ConstructAnalyticsAttribute :AttributeName "AveragePlayerGameplayTimeOnPoi" :IntValue -1 :floatValue _time_2 :BoolValue false))
                (Utilities|Array|Add (Variables|Default|GetAnalyticsEventAttributes) _returnvalue_9)
                (Fortnite|FortShippingLog self (Utilities|String|Append "AveragePlayerGameplayTimeOnPoi: " (Utilities|String|ToString(Float) _time_2))))
              (:then_2
                (bind _returnvalue_10 (Analytics|ConstructAnalyticsAttribute :AttributeName "TotalDamageDoneOutsidePoi" :IntValue -1 :floatValue _totalpawndamagedoneoutsidepoi :BoolValue false))
                (Utilities|Array|Add (Variables|Default|GetAnalyticsEventAttributes) _returnvalue_10)
                (Fortnite|FortShippingLog self (Utilities|String|Append "TotalDamageDoneOutsidePoi: " (Utilities|String|ToString(Float) _totalpawndamagedoneoutsidepoi)))
                (bind _returnvalue_11 (Analytics|ConstructAnalyticsAttribute :AttributeName "TotalPlayerPawnDeathsOnPoi" :IntValue _totalpawndeaths :floatValue -1.0 :BoolValue false))
                (Utilities|Array|Add (Variables|Default|GetAnalyticsEventAttributes) _returnvalue_11)
                (Fortnite|FortShippingLog self (Utilities|String|Append "TotalPlayerPawnDeathsOnPoi: " (Utilities|String|ToString(Integer) _totalpawndeaths)))
                (bind _returnvalue_12 (Analytics|ConstructAnalyticsAttribute :AttributeName "UniquePlayerBotsOnPoi" :IntValue _returnvalue_3 :floatValue -1.0 :BoolValue false))
                (Utilities|Array|Add (Variables|Default|GetAnalyticsEventAttributes) _returnvalue_12)
                (Fortnite|FortShippingLog self (Utilities|String|Append "UniquePlayerBotsOnPoi: " (Utilities|String|ToString(Integer) _returnvalue_3)))
                (bind _returnvalue_13 (Analytics|ConstructAnalyticsAttribute :AttributeName "PoiBannerCaptured" :IntValue -1 :floatValue -1.0 :BoolValue _returnvalue_4))
                (Utilities|Array|Add (Variables|Default|GetAnalyticsEventAttributes) _returnvalue_13)
                (Fortnite|FortShippingLog self (Utilities|String|Append "PoiBannerCaptured: " (Utilities|String|ToString(Boolean) _returnvalue_4))))))
          (:then_1
            (CallFunction|FireEventSessionFinished)))))))
